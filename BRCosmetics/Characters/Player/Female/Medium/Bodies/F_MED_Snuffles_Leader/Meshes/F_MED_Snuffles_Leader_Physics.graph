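# Physics asset for the female medium Snuffles Leader body cosmetic, covering its dangling arm bandana and left pouch, likely for secondary motion.
# Each accessory is a box body constrained with angular limits to a tiny capsule on its base bone, with the bandana's twist kept tight.
# Both base bones and the left thigh link to the pelvis with free angular motion.
# Pelvis and left thigh capsules provide body volume.
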PhysicsAsset /BRCosmetics/Characters/Player/Female/Medium/Bodies/F_MED_Snuffles_Leader/Meshes/F_MED_Snuffles_Leader_Physics
bodies {
 dyn_arm_bandana [{shapeType=Box, extentX=6.0173, extentY=2.8714, extentZ=7.1828, center=<5.5183 0.8231 -0.8434>, rotation=<rot 0 -35.0668 -87.132>}]
 dyn_arm_bandana_base [{shapeType=Capsule, length=1, radius=1}]
 dyn_pouch_base_l [{shapeType=Capsule, length=1, radius=1}]
 dyn_pouch_l [{shapeType=Box, extentX=7.8091, extentY=4.4234, extentZ=17.9415, center=<-0.5308 0.4181 3.0195>, rotation=<rot 0 93.2304 -86.0736>}]
 pelvis [{shapeType=Capsule, length=9.1063, radius=10.8983, center=<1.0192 -1.376 0.0104>, rotation=<rot 0 -6.1325 -179.2519>}]
 thigh_l [{shapeType=Capsule, length=24.5, radius=8.75, center=<-16.9031 0.0004 0>, rotation=<rot 90 0 0>}]
}
constraints [
 {bone1Name=dyn_arm_bandana, bone2Name=dyn_arm_bandana_base, swing1LimitDegrees=40, swing2LimitDegrees=40, twistLimitDegrees=5}
 {
  bone1Name=dyn_arm_bandana_base
  bone2Name=pelvis
  swing1LimitDegrees=45
  swing1Motion=Free
  swing2LimitDegrees=45
  swing2Motion=Free
  twistLimitDegrees=45
  twistMotion=Free
 }
 {
  bone1Name=thigh_l
  bone2Name=pelvis
  swing1LimitDegrees=45
  swing1Motion=Free
  swing2LimitDegrees=45
  swing2Motion=Free
  twistLimitDegrees=45
  twistMotion=Free
 }
 {
  bone1Name=dyn_pouch_base_l
  bone2Name=pelvis
  swing1LimitDegrees=45
  swing1Motion=Free
  swing2LimitDegrees=45
  swing2Motion=Free
  twistLimitDegrees=45
  twistMotion=Free
 }
 {bone1Name=dyn_pouch_l, bone2Name=dyn_pouch_base_l, swing1LimitDegrees=20, swing2LimitDegrees=45, twistLimitDegrees=45}
]
