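# Physics asset for the face accessory part of the female medium Snuffles Leader cosmetic body, covering dynamic scarf, hair and neck-front bones.
# Left and right scarves are five-box chains with limited swing and tight twist, rooted in small spheres linked freely to spine_05.
# Two hair capsules hang from the head and one neck-front capsule from spine_05, all angle-limited.
# Pelvis, spine, head and thigh capsules form a body proxy joined by free-rotation constraints.
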PhysicsAsset /BRCosmetics/Characters/Player/Female/Medium/Bodies/F_MED_Snuffles_Leader/Meshes/Parts/F_MED_Snuffles_Leader_FaceAcc_Physics
bodies {
 dyn_hair_a [{shapeType=Capsule, length=3.5, radius=1.5, center=<1.6996 0 0>, rotation=<rot 90 0 0>}]
 dyn_hair_b [{shapeType=Capsule, length=3.5, radius=1.5, center=<1.6996 0 0>, rotation=<rot 90 0 0>}]
 dyn_neck_front [{shapeType=Capsule, length=3.25, radius=2.75, center=<3.8934 0 0>, rotation=<rot 90 0 0>}]
 dyn_scarf_a_base_l [{shapeType=Sphere, radius=1}]
 dyn_scarf_a_base_r [{shapeType=Sphere, radius=1}]
 dyn_scarf_a_l [{shapeType=Box, extentX=7.5, extentY=2, extentZ=6, center=<3.1593 0.4797 -0.0001>, rotation=<rot 0 0 -2.9428>}]
 dyn_scarf_a_r [{shapeType=Box, extentX=8, extentY=2.5, extentZ=7, center=<-3.9 -0.5917 0.0145>, rotation=<rot -0.2075 -3.0017 -3.9505>}]
 dyn_scarf_b_l [{shapeType=Box, extentX=7.5, extentY=2, extentZ=6.5, center=<3.2684 -0.4867 0.0001>, rotation=<rot 0 -16.9329 0>}]
 dyn_scarf_b_r [{shapeType=Box, extentX=7, extentY=2, extentZ=8, center=<-2.9947 0.5831 0>, rotation=<rot 0 -10.422 0>}]
 dyn_scarf_c_l [{shapeType=Box, extentX=6.5, extentY=2.5, extentZ=6.5, center=<3.4509 0.0001 0>}]
 dyn_scarf_c_r [{shapeType=Box, extentX=8.5, extentY=2.5, extentZ=6.5, center=<-4.3798 0 0>}]
 dyn_scarf_d_l [{shapeType=Box, extentX=5, extentY=2.5, extentZ=6.5, center=<2.3316 0 0>}]
 dyn_scarf_d_r [{shapeType=Box, extentX=6, extentY=2.5, extentZ=6.5, center=<-3.2705 0 -0.0001>}]
 dyn_scarf_e_l [{shapeType=Box, extentX=7, extentY=2.5, extentZ=6.5, center=<3.4953 0.0001 0>}]
 dyn_scarf_e_r [{shapeType=Box, extentX=6, extentY=2, extentZ=7, center=<-3.2549 0 -0.0003>}]
 head [{shapeType=Capsule, length=1, radius=1}]
 pelvis [{shapeType=Capsule, length=9.255, radius=12.255, center=<0 -1.8123 0>, rotation=<rot 0 -90 0>}]
 spine_02 [{shapeType=Capsule, length=8.5, radius=11.5, center=<0 -0.3138 0>}]
 spine_04 [{shapeType=Capsule, length=19.5, radius=10.5, center=<-2.1604 2.5757 0>, rotation=<rot 0 -9.6459 0>}]
 spine_05 [{shapeType=Capsule, length=19, radius=5.25}]
 thigh_l [{shapeType=Capsule, length=31, radius=9, center=<-21.1932 -0.0002 0>, rotation=<rot 90 0 0>}]
 thigh_r [{shapeType=Capsule, length=31, radius=9, center=<21.1932 0.0002 0>, rotation=<rot 90 0 180>}]
}
constraints [
 {bone1Name=dyn_scarf_a_r, bone2Name=dyn_scarf_a_base_r, swing1LimitDegrees=45, swing2LimitDegrees=45, twistLimitDegrees=10}
 {bone1Name=dyn_scarf_b_r, bone2Name=dyn_scarf_a_r, swing1LimitDegrees=45, swing2LimitDegrees=45, twistLimitDegrees=10}
 {bone1Name=dyn_scarf_c_r, bone2Name=dyn_scarf_b_r, swing1LimitDegrees=45, swing2LimitDegrees=45, twistLimitDegrees=10}
 {bone1Name=dyn_scarf_d_r, bone2Name=dyn_scarf_c_r, swing1LimitDegrees=45, swing2LimitDegrees=45, twistLimitDegrees=10}
 {bone1Name=dyn_scarf_e_r, bone2Name=dyn_scarf_d_r, swing1LimitDegrees=45, swing2LimitDegrees=45, twistLimitDegrees=10}
 {bone1Name=dyn_scarf_a_l, bone2Name=dyn_scarf_a_base_l, swing1LimitDegrees=45, swing2LimitDegrees=45, twistLimitDegrees=10}
 {bone1Name=dyn_scarf_b_l, bone2Name=dyn_scarf_a_l, swing1LimitDegrees=45, swing2LimitDegrees=45, twistLimitDegrees=10}
 {bone1Name=dyn_scarf_c_l, bone2Name=dyn_scarf_b_l, swing1LimitDegrees=45, swing2LimitDegrees=45, twistLimitDegrees=10}
 {bone1Name=dyn_scarf_d_l, bone2Name=dyn_scarf_c_l, swing1LimitDegrees=45, swing2LimitDegrees=45, twistLimitDegrees=10}
 {bone1Name=dyn_scarf_e_l, bone2Name=dyn_scarf_d_l, swing1LimitDegrees=45, swing2LimitDegrees=45, twistLimitDegrees=10}
 {
  bone1Name=dyn_scarf_a_base_r
  bone2Name=spine_05
  swing1LimitDegrees=45
  swing1Motion=Free
  swing2LimitDegrees=45
  swing2Motion=Free
  twistLimitDegrees=45
  twistMotion=Free
 }
 {
  bone1Name=dyn_scarf_a_base_l
  bone2Name=spine_05
  swing1LimitDegrees=45
  swing1Motion=Free
  swing2LimitDegrees=45
  swing2Motion=Free
  twistLimitDegrees=45
  twistMotion=Free
 }
 {
  bone1Name=spine_04
  bone2Name=spine_02
  swing1LimitDegrees=45
  swing1Motion=Free
  swing2LimitDegrees=45
  swing2Motion=Free
  twistLimitDegrees=45
  twistMotion=Free
 }
 {
  bone1Name=spine_02
  bone2Name=pelvis
  swing1LimitDegrees=45
  swing1Motion=Free
  swing2LimitDegrees=45
  swing2Motion=Free
  twistLimitDegrees=45
  twistMotion=Free
 }
 {
  bone1Name=spine_05
  bone2Name=spine_04
  swing1LimitDegrees=45
  swing1Motion=Free
  swing2LimitDegrees=45
  swing2Motion=Free
  twistLimitDegrees=45
  twistMotion=Free
 }
 {
  bone1Name=head
  bone2Name=spine_05
  swing1LimitDegrees=45
  swing1Motion=Free
  swing2LimitDegrees=45
  swing2Motion=Free
  twistLimitDegrees=45
  twistMotion=Free
 }
 {bone1Name=dyn_neck_front, bone2Name=spine_05, swing1LimitDegrees=30, swing2LimitDegrees=30, twistLimitDegrees=5}
 {bone1Name=dyn_hair_a, bone2Name=head, swing1LimitDegrees=25, swing2LimitDegrees=25, twistLimitDegrees=10}
 {bone1Name=dyn_hair_b, bone2Name=head, swing1LimitDegrees=25, swing2LimitDegrees=25, twistLimitDegrees=10}
 {
  bone1Name=thigh_l
  bone2Name=pelvis
  swing1LimitDegrees=45
  swing1Motion=Free
  swing2LimitDegrees=45
  swing2Motion=Free
  twistLimitDegrees=45
  twistMotion=Free
 }
 {
  bone1Name=thigh_r
  bone2Name=pelvis
  swing1LimitDegrees=45
  swing1Motion=Free
  swing2LimitDegrees=45
  swing2Motion=Free
  twistLimitDegrees=45
  twistMotion=Free
 }
]
