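# Post-copy AnimBP for the Snuffles Leader female face accessory mesh, animating scarf, brows, ears and face marks.
# Copies the parent mesh pose, applies scarf constraints (some weighted by DynamicsDisabledAlpha), head-forward neckscarf tilt and brow noise offsets, then caches it.
# bCanPlayCustomAnimations blends in variable-driven bone scaling before RigidBody physics and spring AnimDynamics on both ears.
# The event graph is empty, so driving values are set elsewhere.
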AnimBlueprint /BRCosmetics/Characters/Player/Female/Medium/Bodies/F_MED_Snuffles_Leader/Meshes/Parts/F_MED_Snuffles_Leader_FaceAcc_AnimBP
target_skeleton=&/F_MED_Snuffles_Leader_FaceAcc_Skeleton
graphs [
 AnimGraph {nodes [
   n1 = "Transform (Modify) Bone - Bone: dyn_brow_base_base"(
    ComponentPose=Local To Component(LocalPose=Use cached pose 'PreAnime'())
    Alpha=1
    Translation="0,0,0"
    Rotation="0,0,0"
    Scale=Get HitBoneScale() -> (HitBoneScale:Vector)
   ) {
    rotationMode=BMM_Additive
    rotationSpace=BCS_BoneSpace
    scaleMode=BMM_Additive
    scaleSpace=BCS_BoneSpace
    translationMode=BMM_Additive
    translationSpace=BCS_BoneSpace
    alphaBoolBlend {blendOption=Linear}
    alphaScaleBiasClamp {inRange {max=0, min=-40}, outRange {max=0, min=1}}
    boneToModify {boneName=dyn_brow_base_base}
   }
   n2 = "Transform (Modify) Bone - Bone: dyn_shap"(ComponentPose=@n1, Alpha=1, Translation="0,0,0", Rotation="35,0,0", Scale=Get HashTagBoneScale() -> (HashTagBoneScale:Vector)) {
    rotationMode=BMM_Additive
    rotationSpace=BCS_BoneSpace
    scaleMode=BMM_Additive
    scaleSpace=BCS_BoneSpace
    translationMode=BMM_Additive
    translationSpace=BCS_BoneSpace
    alphaBoolBlend {blendOption=Linear}
    alphaScaleBiasClamp {inRange {max=0, min=-40}, outRange {max=0, min=1}}
    boneToModify {boneName=dyn_shap}
    rotation {pitch=35}
   }
   n3 = Get CheekLineBoneScale() -> (CheekLineBoneScale:Vector)
   n4 = "Transform (Modify) Bone - Bone: dyn_cheek_line_l"(ComponentPose=@n2, Alpha=1, Translation="0,0,0", Rotation="0,0,0", Scale=@n3) {
    rotationMode=BMM_Additive
    rotationSpace=BCS_BoneSpace
    scaleMode=BMM_Additive
    scaleSpace=BCS_BoneSpace
    translationMode=BMM_Additive
    translationSpace=BCS_BoneSpace
    alphaBoolBlend {blendOption=Linear}
    alphaScaleBiasClamp {inRange {max=0, min=-40}, outRange {max=0, min=1}}
    boneToModify {boneName=dyn_cheek_line_l}
   }
   n5 = "Transform (Modify) Bone - Bone: dyn_cheek_rine_r"(ComponentPose=@n4, Alpha=1, Translation="0,0,0", Rotation="0,0,0", Scale=@n3) {
    rotationMode=BMM_Additive
    rotationSpace=BCS_BoneSpace
    scaleMode=BMM_Additive
    scaleSpace=BCS_BoneSpace
    translationMode=BMM_Additive
    translationSpace=BCS_BoneSpace
    alphaBoolBlend {blendOption=Linear}
    alphaScaleBiasClamp {inRange {max=0, min=-40}, outRange {max=0, min=1}}
    boneToModify {boneName=dyn_cheek_rine_r}
   }
   n6 = Get EarlineBoneScale() -> (EarlineBoneScale:Vector)
   n7 = "Transform (Modify) Bone - Bone: dyn_earbrow_base_l"(ComponentPose=@n5, Alpha=1, Translation="0,0,0", Rotation="0,0,0", Scale=@n6) {
    rotationMode=BMM_Additive
    rotationSpace=BCS_BoneSpace
    scaleMode=BMM_Additive
    scaleSpace=BCS_BoneSpace
    translationMode=BMM_Additive
    translationSpace=BCS_BoneSpace
    alphaBoolBlend {blendOption=Linear}
    alphaScaleBiasClamp {inRange {max=0, min=-40}, outRange {max=0, min=1}}
    boneToModify {boneName=dyn_earbrow_base_l}
   }
   n8 = "Transform (Modify) Bone - Bone: dyn_earbrow_base_r"(ComponentPose=@n7, Alpha=1, Translation="0,0,0", Rotation="0,0,0", Scale=@n6) {
    rotationMode=BMM_Additive
    rotationSpace=BCS_BoneSpace
    scaleMode=BMM_Additive
    scaleSpace=BCS_BoneSpace
    translationMode=BMM_Additive
    translationSpace=BCS_BoneSpace
    alphaBoolBlend {blendOption=Linear}
    alphaScaleBiasClamp {inRange {max=0, min=-40}, outRange {max=0, min=1}}
    boneToModify {boneName=dyn_earbrow_base_r}
   }
   n9 = Component To Local(ComponentPose=@n8)
   n10 = "Transform (Modify) Bone - Bone: dyn_brow_base_base"(ComponentPose=Local To Component(LocalPose=Use cached pose 'PreAnime'()), Alpha=1, Translation="0,0,0", Rotation="0,0,0", Scale="1,1,1") {
    rotationMode=BMM_Additive
    rotationSpace=BCS_BoneSpace
    scaleMode=BMM_Additive
    scaleSpace=BCS_BoneSpace
    translationMode=BMM_Additive
    translationSpace=BCS_BoneSpace
    alphaBoolBlend {blendOption=Linear}
    alphaScaleBiasClamp {inRange {max=0, min=-40}, outRange {max=0, min=1}}
    boneToModify {boneName=dyn_brow_base_base}
   }
   n11 = "Transform (Modify) Bone - Bone: dyn_shap"(ComponentPose=@n10, Alpha=1, Translation="0,0,0", Rotation="35,0,0", Scale="1,1,1") {
    rotationMode=BMM_Additive
    rotationSpace=BCS_BoneSpace
    scaleMode=BMM_Additive
    scaleSpace=BCS_BoneSpace
    translationMode=BMM_Additive
    translationSpace=BCS_BoneSpace
    alphaBoolBlend {blendOption=Linear}
    alphaScaleBiasClamp {inRange {max=0, min=-40}, outRange {max=0, min=1}}
    boneToModify {boneName=dyn_shap}
    rotation {pitch=35}
   }
   n12 = "Transform (Modify) Bone - Bone: dyn_cheek_line_l"(ComponentPose=@n11, Alpha=1, Translation="0,0,0", Rotation="0,0,0", Scale="1,1,1") {
    rotationMode=BMM_Additive
    rotationSpace=BCS_BoneSpace
    scaleMode=BMM_Additive
    scaleSpace=BCS_BoneSpace
    translationMode=BMM_Additive
    translationSpace=BCS_BoneSpace
    alphaBoolBlend {blendOption=Linear}
    alphaScaleBiasClamp {inRange {max=0, min=-40}, outRange {max=0, min=1}}
    boneToModify {boneName=dyn_cheek_line_l}
   }
   n13 = "Transform (Modify) Bone - Bone: dyn_cheek_rine_r"(ComponentPose=@n12, Alpha=1, Translation="0,0,0", Rotation="0,0,0", Scale="1,1,1") {
    rotationMode=BMM_Additive
    rotationSpace=BCS_BoneSpace
    scaleMode=BMM_Additive
    scaleSpace=BCS_BoneSpace
    translationMode=BMM_Additive
    translationSpace=BCS_BoneSpace
    alphaBoolBlend {blendOption=Linear}
    alphaScaleBiasClamp {inRange {max=0, min=-40}, outRange {max=0, min=1}}
    boneToModify {boneName=dyn_cheek_rine_r}
   }
   n14 = "Transform (Modify) Bone - Bone: dyn_earbrow_base_l"(ComponentPose=@n13, Alpha=1, Translation="0,0,0", Rotation="0,0,0", Scale="1,1,1") {
    rotationMode=BMM_Additive
    rotationSpace=BCS_BoneSpace
    scaleMode=BMM_Additive
    scaleSpace=BCS_BoneSpace
    translationMode=BMM_Additive
    translationSpace=BCS_BoneSpace
    alphaBoolBlend {blendOption=Linear}
    alphaScaleBiasClamp {inRange {max=0, min=-40}, outRange {max=0, min=1}}
    boneToModify {boneName=dyn_earbrow_base_l}
   }
   n15 = "Transform (Modify) Bone - Bone: dyn_earbrow_base_r"(ComponentPose=@n14, Alpha=1, Translation="0,0,0", Rotation="0,0,0", Scale="1,1,1") {
    rotationMode=BMM_Additive
    rotationSpace=BCS_BoneSpace
    scaleMode=BMM_Additive
    scaleSpace=BCS_BoneSpace
    translationMode=BMM_Additive
    translationSpace=BCS_BoneSpace
    alphaBoolBlend {blendOption=Linear}
    alphaScaleBiasClamp {inRange {max=0, min=-40}, outRange {max=0, min=1}}
    boneToModify {boneName=dyn_earbrow_base_r}
   }
   n16 = Component To Local(ComponentPose=@n15)
   n17 = Get bCanPlayCustomAnimations() -> (bCanPlayCustomAnimations:bool)
   n18 = RigidBody(ComponentPose=Local To Component(LocalPose=Blend Poses by bool(BlendPose_0=@n16, BlendPose_1=@n9, bActiveValue=@n17) {blendType=Linear})) {
    bClampLinearTranslationLimitToRefPose=true
    bOverrideWorldGravity=true
    overridePhysicsAsset=&/F_MED_Snuffles_Leader_FaceAcc_Physics
    simulationSpace=BaseBoneSpace
    alphaBoolBlend {blendOption=Linear}
    baseBoneRef {boneName=spine_02}
    componentLinearAccScale=<0.4 0.1 0.1>
    componentLinearVelScale=<0.4 0.1 0.1>
    overrideWorldGravity {z=-980}
   }
   n19 = Anim Dynamics - Bone: dyn_ear_l!(ComponentPose=@n18, Alpha=1) -> (Pose) {
    angularDampingOverride=1
    angularSpringConstant=4000
    bAngularSpring=true
    bOverrideAngularDamping=true
    bUseGravityOverride=true
    gravityScale=0.35
    simulationSpace=BoneRelative
    alphaBoolBlend {blendOption=Linear}
    boundBone {boneName=dyn_ear_l}
    componentAppliedLinearAccClamp=<10000 10000 10000>
    componentLinearAccScale=<1 1 1>
    componentLinearVelScale=<1 1 1>
    gravityOverride {z=-980}
    physicsBodyDefinitions [{
      boundBone {boneName=dyn_ear_l}
      boxExtents=<18 18 18>
      constraintSetup {angularLimitsMax=<0 15 20>, angularLimitsMin=<0 -15 -20>, angularTarget=<1 0 0>}
      localJointOffset=<1 0 0>
     }]
    relativeSpaceBone {boneName=spine_03}
   }
   n20 = Anim Dynamics - Bone: dyn_ear_r!(ComponentPose=@n19, Alpha=1) -> (Pose) {
    angularDampingOverride=1
    angularSpringConstant=4000
    bAngularSpring=true
    bOverrideAngularDamping=true
    bUseGravityOverride=true
    gravityScale=0.35
    simulationSpace=BoneRelative
    alphaBoolBlend {blendOption=Linear}
    boundBone {boneName=dyn_ear_r}
    componentAppliedLinearAccClamp=<10000 10000 10000>
    componentLinearAccScale=<1 1 1>
    componentLinearVelScale=<1 1 1>
    gravityOverride {z=-980}
    physicsBodyDefinitions [{
      boundBone {boneName=dyn_ear_r}
      boxExtents=<18 18 18>
      constraintSetup {angularLimitsMax=<0 15 20>, angularLimitsMin=<0 -15 -20>, angularTarget=<1 0 0>}
      localJointOffset=<-1 0 0>
     }]
    relativeSpaceBone {boneName=spine_03}
   }
   Output Pose(Result=Component To Local(ComponentPose=@n20)) {layerGroup=None}
   n21 = Get MeshToCopy() -> (MeshToCopy:Skeletal Mesh Component Object Reference)
   n22 = "Constraint - dyn_scarf_a_base_l (PPP)"!(
    ComponentPose=Local To Component(LocalPose=Copy Pose From Mesh(SourceMeshComponent=@n21) {bUseAttachedParent=true})
    Alpha=1
    ConstraintWeights_0=0.6649
    ConstraintWeights_1:float=0.3093
    ConstraintWeights_2:float=0.0258
   ) -> (Pose) {
    alphaBoolBlend {blendOption=Linear}
    boneToModify {boneName=dyn_scarf_a_base_l}
    constraintSetup [{targetBone {boneName=spine_04}}, {targetBone {boneName=spine_05}}, {targetBone {boneName=clavicle_l}}]
    constraintWeights [1,1,1]
   }
   n23 = "Constraint - dyn_scarf_a_base_r (PPP)"!(ComponentPose=@n22, Alpha=1, ConstraintWeights_0=0.0259, ConstraintWeights_1:float=0.665, ConstraintWeights_2:float=0.3091) -> (Pose) {
    alphaBoolBlend {blendOption=Linear}
    boneToModify {boneName=dyn_scarf_a_base_r}
    constraintSetup [{targetBone {boneName=clavicle_r}}, {targetBone {boneName=spine_04}}, {targetBone {boneName=spine_05}}]
    constraintWeights [1,1,1]
   }
   n24 = Get DynamicsDisabledAlpha() -> (DynamicsDisabledAlpha:float)
   n25 = "Constraint - dyn_scarf_b_l (PP)"!(ComponentPose=@n23, Alpha=@n24, ConstraintWeights_0=0.9596, ConstraintWeights_1:float=0.0404) -> (Pose) {
    alphaBoolBlend {blendOption=Linear}
    boneToModify {boneName=dyn_scarf_b_l}
    constraintSetup [{targetBone {boneName=spine_04}}, {targetBone {boneName=spine_05}}]
    constraintWeights [1,1]
   }
   n26 = "Constraint - dyn_scarf_c_l (PP)"!(ComponentPose=@n25, Alpha=@n24, ConstraintWeights_0=0.038, ConstraintWeights_1:float=0.962) -> (Pose) {
    alphaBoolBlend {blendOption=Linear}
    boneToModify {boneName=dyn_scarf_c_l}
    constraintSetup [{targetBone {boneName=spine_03}}, {targetBone {boneName=spine_04}}]
    constraintWeights [1,1]
   }
   n27 = "Constraint - dyn_scarf_d_l (PPPPP)"!(
    ComponentPose=@n26
    Alpha=@n24
    ConstraintWeights_0=0.1669
    ConstraintWeights_1:float=0.3245
    ConstraintWeights_2:float=0.1394
    ConstraintWeights_3:float=0.2438
    ConstraintWeights_4:float=0.1254
   ) -> (Pose) {
    alphaBoolBlend {blendOption=Linear}
    boneToModify {boneName=dyn_scarf_d_l}
    constraintSetup [
     {targetBone {boneName=spine_01}}
     {targetBone {boneName=spine_02}}
     {targetBone {boneName=spine_03}}
     {targetBone {boneName=spine_04}}
     {targetBone {boneName=pelvis}}
    ]
    constraintWeights [1,1,1,1,1]
   }
   n28 = "Constraint - dyn_scarf_e_l (PPPP)"!(
    ComponentPose=@n27
    Alpha=@n24
    ConstraintWeights_0=0.1141
    ConstraintWeights_1:float=0.1999
    ConstraintWeights_2:float=0.0546
    ConstraintWeights_3:float=0.6313
   ) -> (Pose) {
    alphaBoolBlend {blendOption=Linear}
    boneToModify {boneName=dyn_scarf_e_l}
    constraintSetup [{targetBone {boneName=spine_01}}, {targetBone {boneName=spine_02}}, {targetBone {boneName=spine_04}}, {targetBone {boneName=pelvis}}]
    constraintWeights [1,1,1,1]
   }
   n29 = Get DynamicsDisabledAlpha()
   n30 = "Constraint - dyn_scarf_b_r (PP)"!(ComponentPose=@n28, Alpha=@n29, ConstraintWeights_0=0.9596, ConstraintWeights_1:float=0.0404) -> (Pose) {
    alphaBoolBlend {blendOption=Linear}
    boneToModify {boneName=dyn_scarf_b_r}
    constraintSetup [{targetBone {boneName=spine_04}}, {targetBone {boneName=spine_05}}]
    constraintWeights [1,1]
   }
   n31 = "Constraint - dyn_scarf_c_r (PP)"!(ComponentPose=@n30, Alpha=@n29, ConstraintWeights_0=0.038, ConstraintWeights_1:float=0.962) -> (Pose) {
    alphaBoolBlend {blendOption=Linear}
    boneToModify {boneName=dyn_scarf_c_r}
    constraintSetup [{targetBone {boneName=spine_03}}, {targetBone {boneName=spine_04}}]
    constraintWeights [1,1]
   }
   n32 = "Constraint - dyn_scarf_d_r (PPPPP)"!(
    ComponentPose=@n31
    Alpha=@n29
    ConstraintWeights_0=0.1859
    ConstraintWeights_1:float=0.3371
    ConstraintWeights_2:float=0.0511
    ConstraintWeights_3:float=0.1206
    ConstraintWeights_4:float=0.3053
   ) -> (Pose) {
    alphaBoolBlend {blendOption=Linear}
    boneToModify {boneName=dyn_scarf_d_r}
    constraintSetup [
     {targetBone {boneName=spine_01}}
     {targetBone {boneName=spine_02}}
     {targetBone {boneName=spine_03}}
     {targetBone {boneName=spine_04}}
     {targetBone {boneName=pelvis}}
    ]
    constraintWeights [1,1,1,1,1]
   }
   n33 = "Constraint - dyn_scarf_e_r (P)"!(ComponentPose=@n32, Alpha=@n29, ConstraintWeights_0=1, ConstraintWeights_1:float=0) -> (Pose) {
    alphaBoolBlend {blendOption=Linear}
    boneToModify {boneName=dyn_scarf_e_r}
    constraintSetup [{targetBone {boneName=pelvis}}]
    constraintWeights [1,0]
   }
   n34 = "Constraint - dyn_neckScarf_base (R)"(ComponentPose=@n33, Alpha=0.6, ConstraintWeights_0=1) {
    alphaBoolBlend {blendOption=Linear}
    boneToModify {boneName=dyn_neckScarf_base}
    constraintSetup [{transformType=Rotation, targetBone {boneName=head}}]
    constraintWeights [1]
   }
   n35 = "Constraint - dyn_neckscarf_front_base_l (PP)"!(ComponentPose=@n34, Alpha=0.5, ConstraintWeights_0=0.5, ConstraintWeights_1:float=0.5) -> (Pose) {
    alphaBoolBlend {blendOption=Linear}
    boneToModify {boneName=dyn_neckscarf_front_base_l}
    constraintSetup [{targetBone {boneName=dyn_neckscarf_front}}, {targetBone {boneName=dyn_neckscarf_side_l}}]
    constraintWeights [1,1]
   }
   n36 = "Constraint - dyn_neckscarf_back_base_l (PP)"!(ComponentPose=@n35, Alpha=0.5, ConstraintWeights_0=0.5, ConstraintWeights_1:float=0.5) -> (Pose) {
    alphaBoolBlend {blendOption=Linear}
    boneToModify {boneName=dyn_neckscarf_back_base_l}
    constraintSetup [{targetBone {boneName=dyn_neckscarf_back}}, {targetBone {boneName=dyn_neckscarf_side_l}}]
    constraintWeights [1,1]
   }
   n37 = "Constraint - dyn_neckscarf_front_base_r (PP)"!(ComponentPose=@n36, Alpha=0.5, ConstraintWeights_0=0.5, ConstraintWeights_1:float=0.5) -> (Pose) {
    alphaBoolBlend {blendOption=Linear}
    boneToModify {boneName=dyn_neckscarf_front_base_r}
    constraintSetup [{targetBone {boneName=dyn_neckscarf_front}}, {targetBone {boneName=dyn_neckscarf_side_r}}]
    constraintWeights [1,1]
   }
   n38 = "Constraint - dyn_neckscarf_back_base_l (PP)"(ComponentPose=@n37, Alpha=0.5, ConstraintWeights_0=0.5, ConstraintWeights_1=0.5) {
    alphaBoolBlend {blendOption=Linear}
    boneToModify {boneName=dyn_neckscarf_back_base_l}
    constraintSetup [{targetBone {boneName=dyn_neckscarf_back}}, {targetBone {boneName=dyn_neckscarf_side_r}}]
    constraintWeights [1,1]
   }
   n39 = Get HeadForwardAlpha() -> (HeadForwardAlpha:float)
   n40 = "Transform (Modify) Bone - Bone: dyn_neckscarf_back_base"(ComponentPose=@n38, Alpha=@n39, Translation="0,0,0", Rotation="-10,0,0", Scale="1,1,1") {
    rotationMode=BMM_Additive
    rotationSpace=BCS_BoneSpace
    scaleMode=BMM_Additive
    scaleSpace=BCS_BoneSpace
    translationMode=BMM_Additive
    translationSpace=BCS_BoneSpace
    alphaBoolBlend {blendOption=Linear}
    alphaScaleBiasClamp {bMapRange=true, inRange {max=0, min=-40}, outRange {max=0, min=1}}
    boneToModify {boneName=dyn_neckscarf_back_base}
    rotation {pitch=-10}
   }
   n41 = "Transform (Modify) Bone - Bone: dyn_neckscarf_back_base_l"(ComponentPose=@n40, Alpha=@n39, Translation="0,0,0", Rotation="-25,0,0", Scale="1,1,1") {
    rotationMode=BMM_Additive
    rotationSpace=BCS_BoneSpace
    scaleMode=BMM_Additive
    scaleSpace=BCS_BoneSpace
    translationMode=BMM_Additive
    translationSpace=BCS_BoneSpace
    alphaBoolBlend {blendOption=Linear}
    alphaScaleBiasClamp {bMapRange=true, inRange {max=0, min=-40}, outRange {max=0, min=1}}
    boneToModify {boneName=dyn_neckscarf_back_base_l}
    rotation {pitch=-25}
   }
   n42 = "Transform (Modify) Bone - Bone: dyn_neckscarf_side_base_l"(ComponentPose=@n41, Alpha=@n39, Translation="0,0,0", Rotation="-20,0,0", Scale="1,1,1") {
    rotationMode=BMM_Additive
    rotationSpace=BCS_BoneSpace
    scaleMode=BMM_Additive
    scaleSpace=BCS_BoneSpace
    translationMode=BMM_Additive
    translationSpace=BCS_BoneSpace
    alphaBoolBlend {blendOption=Linear}
    alphaScaleBiasClamp {bMapRange=true, inRange {max=0, min=-40}, outRange {max=0, min=1}}
    boneToModify {boneName=dyn_neckscarf_side_base_l}
    rotation {pitch=-20}
   }
   n43 = "Transform (Modify) Bone - Bone: dyn_neckscarf_front_base_l"(ComponentPose=@n42, Alpha=@n39, Translation="0,0,0", Rotation="-20,0,0", Scale="1,1,1") {
    rotationMode=BMM_Additive
    rotationSpace=BCS_BoneSpace
    scaleMode=BMM_Additive
    scaleSpace=BCS_BoneSpace
    translationMode=BMM_Additive
    translationSpace=BCS_BoneSpace
    alphaBoolBlend {blendOption=Linear}
    alphaScaleBiasClamp {bMapRange=true, inRange {max=0, min=-40}, outRange {max=0, min=1}}
    boneToModify {boneName=dyn_neckscarf_front_base_l}
    rotation {pitch=-20}
   }
   n44 = Get HeadForwardAlpha()
   n45 = "Transform (Modify) Bone - Bone: dyn_neckscarf_back_base_r"(ComponentPose=@n43, Alpha=@n44, Translation="0,0,0", Rotation="-25,0,0", Scale="1,1,1") {
    rotationMode=BMM_Additive
    rotationSpace=BCS_BoneSpace
    scaleMode=BMM_Additive
    scaleSpace=BCS_BoneSpace
    translationMode=BMM_Additive
    translationSpace=BCS_BoneSpace
    alphaBoolBlend {blendOption=Linear}
    alphaScaleBiasClamp {bMapRange=true, inRange {max=0, min=-40}, outRange {max=0, min=1}}
    boneToModify {boneName=dyn_neckscarf_back_base_r}
    rotation {pitch=-25}
   }
   n46 = "Transform (Modify) Bone - Bone: dyn_neckscarf_side_base_r"(ComponentPose=@n45, Alpha=@n44, Translation="0,0,0", Rotation="-20,0,0", Scale="1,1,1") {
    rotationMode=BMM_Additive
    rotationSpace=BCS_BoneSpace
    scaleMode=BMM_Additive
    scaleSpace=BCS_BoneSpace
    translationMode=BMM_Additive
    translationSpace=BCS_BoneSpace
    alphaBoolBlend {blendOption=Linear}
    alphaScaleBiasClamp {bMapRange=true, inRange {max=0, min=-40}, outRange {max=0, min=1}}
    boneToModify {boneName=dyn_neckscarf_side_base_r}
    rotation {pitch=-20}
   }
   n47 = "Transform (Modify) Bone - Bone: dyn_neckscarf_front_base_r"(ComponentPose=@n46, Alpha=@n44, Translation="0,0,0", Rotation="-20,0,0", Scale="1,1,1") {
    rotationMode=BMM_Additive
    rotationSpace=BCS_BoneSpace
    scaleMode=BMM_Additive
    scaleSpace=BCS_BoneSpace
    translationMode=BMM_Additive
    translationSpace=BCS_BoneSpace
    alphaBoolBlend {blendOption=Linear}
    alphaScaleBiasClamp {bMapRange=true, inRange {max=0, min=-40}, outRange {max=0, min=1}}
    boneToModify {boneName=dyn_neckscarf_front_base_r}
    rotation {pitch=-20}
   }
   n48 = Get EyebrowsNoiseVector() -> (EyebrowsNoiseVector:Vector)
   n49 = "Transform (Modify) Bone - Bone: dyn_brow_up_base_l"(ComponentPose=@n47, Alpha=1, Translation=@n48, Rotation="0,0,0", Scale="1,1,1") {
    rotationMode=BMM_Additive
    rotationSpace=BCS_BoneSpace
    scaleMode=BMM_Additive
    scaleSpace=BCS_BoneSpace
    translationMode=BMM_Additive
    translationSpace=BCS_BoneSpace
    alphaBoolBlend {blendOption=Linear}
    alphaScaleBiasClamp {inRange {max=0, min=-40}, outRange {max=0, min=1}}
    boneToModify {boneName=dyn_brow_up_base_l}
   }
   n50 = "Transform (Modify) Bone - Bone: dyn_brow_up_base_r"(ComponentPose=@n49, Alpha=1, Translation=@n48, Rotation="0,0,0", Scale="1,1,1") {
    rotationMode=BMM_Additive
    rotationSpace=BCS_BoneSpace
    scaleMode=BMM_Additive
    scaleSpace=BCS_BoneSpace
    translationMode=BMM_Additive
    translationSpace=BCS_BoneSpace
    alphaBoolBlend {blendOption=Linear}
    alphaScaleBiasClamp {inRange {max=0, min=-40}, outRange {max=0, min=1}}
    boneToModify {boneName=dyn_brow_up_base_r}
   }
   n51 = "Transform (Modify) Bone - Bone: dyn_brow_dn_base_l"(ComponentPose=@n50, Alpha=1, Translation=@n48, Rotation="0,0,0", Scale="1,1,1") {
    rotationMode=BMM_Additive
    rotationSpace=BCS_BoneSpace
    scaleMode=BMM_Additive
    scaleSpace=BCS_BoneSpace
    translationMode=BMM_Additive
    translationSpace=BCS_BoneSpace
    alphaBoolBlend {blendOption=Linear}
    alphaScaleBiasClamp {inRange {max=0, min=-40}, outRange {max=0, min=1}}
    boneToModify {boneName=dyn_brow_dn_base_l}
   }
   n52 = "Transform (Modify) Bone - Bone: dyn_brow_dn_base_r"(ComponentPose=@n51, Alpha=1, Translation=@n48, Rotation="0,0,0", Scale="1,1,1") {
    rotationMode=BMM_Additive
    rotationSpace=BCS_BoneSpace
    scaleMode=BMM_Additive
    scaleSpace=BCS_BoneSpace
    translationMode=BMM_Additive
    translationSpace=BCS_BoneSpace
    alphaBoolBlend {blendOption=Linear}
    alphaScaleBiasClamp {inRange {max=0, min=-40}, outRange {max=0, min=1}}
    boneToModify {boneName=dyn_brow_dn_base_r}
   }
   n53 = "Transform (Modify) Bone - Bone: dyn_earbrowUp_base_l"(ComponentPose=@n52, Alpha=1, Translation=Get EarbrowNoiseVectorA() -> (EarbrowNoiseVectorA:Vector), Rotation="0,0,0", Scale="1,1,1") {
    rotationMode=BMM_Additive
    rotationSpace=BCS_BoneSpace
    scaleMode=BMM_Additive
    scaleSpace=BCS_BoneSpace
    translationMode=BMM_Additive
    translationSpace=BCS_BoneSpace
    alphaBoolBlend {blendOption=Linear}
    alphaScaleBiasClamp {inRange {max=0, min=-40}, outRange {max=0, min=1}}
    boneToModify {boneName=dyn_earbrowUp_base_l}
   }
   n54 = "Transform (Modify) Bone - Bone: dyn_earbrowLo_base_l"(ComponentPose=@n53, Alpha=1, Translation=Get EarbrowNoiseVectorB() -> (EarbrowNoiseVectorB:Vector), Rotation="0,0,0", Scale="1,1,1") {
    rotationMode=BMM_Additive
    rotationSpace=BCS_BoneSpace
    scaleMode=BMM_Additive
    scaleSpace=BCS_BoneSpace
    translationMode=BMM_Additive
    translationSpace=BCS_BoneSpace
    alphaBoolBlend {blendOption=Linear}
    alphaScaleBiasClamp {inRange {max=0, min=-40}, outRange {max=0, min=1}}
    boneToModify {boneName=dyn_earbrowLo_base_l}
   }
   n55 = "Transform (Modify) Bone - Bone: dyn_earbrowUp_base_r"(ComponentPose=@n54, Alpha=1, Translation=Get EarbrowNoiseVectorA(), Rotation="0,0,0", Scale="1,1,1") {
    rotationMode=BMM_Additive
    rotationSpace=BCS_BoneSpace
    scaleMode=BMM_Additive
    scaleSpace=BCS_BoneSpace
    translationMode=BMM_Additive
    translationSpace=BCS_BoneSpace
    alphaBoolBlend {blendOption=Linear}
    alphaScaleBiasClamp {inRange {max=0, min=-40}, outRange {max=0, min=1}}
    boneToModify {boneName=dyn_earbrowUp_base_r}
   }
   n56 = "Transform (Modify) Bone - Bone: dyn_earbrowLo_base_r"(ComponentPose=@n55, Alpha=1, Translation=Get EarbrowNoiseVectorB(), Rotation="0,0,0", Scale="1,1,1") {
    rotationMode=BMM_Additive
    rotationSpace=BCS_BoneSpace
    scaleMode=BMM_Additive
    scaleSpace=BCS_BoneSpace
    translationMode=BMM_Additive
    translationSpace=BCS_BoneSpace
    alphaBoolBlend {blendOption=Linear}
    alphaScaleBiasClamp {inRange {max=0, min=-40}, outRange {max=0, min=1}}
    boneToModify {boneName=dyn_earbrowLo_base_r}
   }
   n57 = "Transform (Modify) Bone - Bone: dyn_shap_base"(ComponentPose=@n56, Alpha=1, Translation="0,0,0", Rotation="0,-55,0", Scale="1,1,1") {
    rotationMode=BMM_Additive
    rotationSpace=BCS_BoneSpace
    scaleMode=BMM_Additive
    scaleSpace=BCS_BoneSpace
    translationMode=BMM_Additive
    translationSpace=BCS_BoneSpace
    alphaBoolBlend {blendOption=Linear}
    alphaScaleBiasClamp {inRange {max=0, min=-40}, outRange {max=0, min=1}}
    boneToModify {boneName=dyn_shap_base}
    rotation {yaw=-55}
   }
   n58 = "Transform (Modify) Bone - Bone: dyn_shap"(ComponentPose=@n57, Alpha=1, Translation="0,-0.5,0", Rotation="35,0,0", Scale="1,1,1") {
    rotationMode=BMM_Additive
    rotationSpace=BCS_BoneSpace
    scaleMode=BMM_Additive
    scaleSpace=BCS_BoneSpace
    translationMode=BMM_Additive
    translationSpace=BCS_BoneSpace
    alphaBoolBlend {blendOption=Linear}
    alphaScaleBiasClamp {inRange {max=0, min=-40}, outRange {max=0, min=1}}
    boneToModify {boneName=dyn_shap}
    rotation {pitch=35}
    translation {y=-0.5}
   }
   Save cached pose 'PreAnime'(Pose=Component To Local(ComponentPose=@n58)) {cacheName=PreAnime}
   Get HitBoneScale()
   Get HashTagBoneScale()
   Get CheekLineBoneScale()
   Get EarlineBoneScale()
   Get PlayerHealth() -> (PlayerHealth:float)
  ]}
 EventGraph {links [], nodes []}
]
variables []
